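; Quest gameplay ability that grants the owning player an external emote while it is active, apparently for an Emerald dance-near-character quest.
; On activation it adds the configured Emote through the player controller's FortControllerComponent_ExternalEmote, and on end it removes it.
; A shared helper casts the avatar to a FortPlayerPawn, caches pawn and component, and creates the component at runtime if absent.
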(event Ability|EventActivateAbility
  ;[1 1 1] Check for External Emote Component if we don't have one create one then add emote
  (|CallParentFunction)
  (bind (_outputpin _emote) (CallFunction|GetOrCreateExternalEmoteComponent))
  (Utilities|IsValid _outputpin
    (:Is%20Valid
      (Class|FortControllerComponentExternalEmote|AddExternalEmoteSimple _outputpin _emote true))))

(event Ability|EventOnEndAbility
  (params (bWasCancelled :type Boolean))
  ;[1 1 1] When everything is done check for External Emote Component, if we don't have one create it then remove emote
  (|CallParentFunction bWasCancelled)
  (bind (_outputpin _emote) (CallFunction|GetOrCreateExternalEmoteComponent))
  (Utilities|IsValid _outputpin
    (:Is%20Valid
      (Class|FortControllerComponentExternalEmote|RemoveExternalEmote _outputpin _emote))))

(event Ability|EventFailedToActivatePassiveAbility
  (params (ActorInfo :type Gameplay%20Ability%20Actor%20Info%20Structure))
  (|CallParentFunction))

(fn GetorCreateExternalEmoteComponent
  (bind _asfort_player_pawn (Utilities|Casting|CastToFortPlayerPawn (Ability|GetAvatarActorfromActorInfo))
    (:then
      (bind _output_get (Variables|Default|SetOwningPlayerPawn _asfort_player_pawn))
      (bind _output_get_2 (Variables|Default|SetFortControllerComponentExternalEmote (Actor|GetComponentbyClass (Fortnite|GetFortPlayerControllerfromActor _output_get) "/Script/FortniteGame.FortControllerComponent_ExternalEmote")))
      (Utilities|IsValid _output_get_2
        (:Is%20Valid
          _)
        (:Is%20Not%20Valid
          (bind _returnvalue (Game|AddComponentbyClass (Fortnite|GetFortPlayerControllerfromActor _output_get) "/Script/FortniteGame.FortControllerComponent_ExternalEmote"))
          (Utilities|IsValid _returnvalue
            (:Is%20Valid
              (Variables|Default|SetFortControllerComponentExternalEmote _returnvalue))))))
    (:CastFailed
      _))
  (return _output_get_2 (Variables|Default|GetEmote)))
